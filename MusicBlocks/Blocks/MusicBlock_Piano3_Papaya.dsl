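; Piano music block Blueprint (Papaya variant) that plays its audio with an animated glow.
; On AnimateDynamicMaterial, a valid dynamic material instance starts the sound and a timeline that drives the GlowColor parameter, then switches to a constant material when finished.
; Without an instance, which the author attributes to too many blocks animating, both sound and glow are skipped.
; Construction defers to the parent class.
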(event EventAnimateDynamicMaterial
  (bind _material_instance_dynamic (Variables|Default|GetMaterial%20Instance%20Dynamic)
    (:then
      (Audio|Components|Audio|Play (Variables|Default|GetAudio) 0.0)
      (bind (_direction _glowintensity) (Timeline|Timeline_1 PlayFromStart)
        (:Update
          (Rendering|Material|SetVectorParameterValue _material_instance_dynamic "GlowColor" _glowintensity :!class /Script/Engine.MaterialInstanceDynamic))
        (:Finished
          (CallFunction|SwitchToConstantMaterial))))
    (:else)) ; Skip animating if we failed to get a dynamic\nmaterial instance (too may things animating)
)

(fn ConstructionScript
  (|CallParentFunction))
